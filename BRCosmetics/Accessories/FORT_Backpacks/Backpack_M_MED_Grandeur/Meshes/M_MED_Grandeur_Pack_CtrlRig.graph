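# Control Rig for the M_MED_Grandeur backpack that drives a bottom strap bone from the rotation of the top strap bone.
# It converts the rotation of dyn_strap_top_1 to Euler angles and remaps the Y angle into a weight.
# That weight scales a fixed Y-axis rotation offset on dyn_strap_bot_1, which grows as the angle drops through the remapped range.
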ControlRigBlueprint /BRCosmetics/Accessories/FORT_Backpacks/Backpack_M_MED_Grandeur/Meshes/M_MED_Grandeur_Pack_CtrlRig
graphs [
 RigVMFunctionLibrary {links [], nodes []}
 RigVMModel {nodes [
   RigUnit_BeginExecution = &RigUnit_BeginExecution()
   n1 = &RigUnit_GetTransform(Item=(Type=Bone,Name="dyn_strap_top_1"))
   &RigUnit_ModifyTransforms(
    ItemToModify=((Item=(Type=Bone),Transform=(Rotation=(X=0,Y=0,Z=0,W=1),Translation=(X=0,Y=0,Z=0),Scale3D=(X=1,Y=1,Z=1))),(Item=(Type=Bone,Name="dyn_strap_bot_1"),Transform=(Rotation=(X=0,Y=0,Z=0,W=1),Translation=(X=0,Y=0,Z=0),Scale3D=(X=1,Y=1,Z=1))))
    Weight=&RigVMFunction_MathFloatRemap(
     SourceMaximum=30
     SourceMinimum=80
     TargetMinimum=0
     Value=&RigVMFunction_MathQuaternionToEuler(RotationOrder=XYZ, Value=@n1.Transform.Rotation).Result.Y
    ).Result
    ItemToModify.1.Transform.Rotation=&RigVMFunction_MathQuaternionFromEuler(Euler=(X=0,Y=-40,Z=0)).Result
    ExecutePin=@RigUnit_BeginExecution.ExecutePin
   )
  ]}
]
hierarchy [
 root {loc=<0 0 0>, rot=<0 0 0 1>}
 pack {parent=root, loc=<0 -0.7316 0>, rot=<0 0 0 1>}
 dyn_strap_MSTR {parent=pack, loc=<0 -3.2783 0>, rot=<0.566 -0.4238 -0.4238 -0.566>}
 dyn_strap_top_1 {parent=dyn_strap_MSTR, loc=<-7.2606 -15.6034 -0.0041>, rot=<0 0.0003 -0.2146 -0.9767>}
 dyn_strap_top_2 {parent=dyn_strap_top_1, loc=<17.1932 0 0>, rot=<-0.0001 -0.0002 -0.2702 -0.9628>}
 dyn_strap_top_3 {parent=dyn_strap_top_2, loc=<10.8297 0 0>, rot=<0 0 -0.1962 -0.9806>}
 dyn_strap_bot_1 {parent=dyn_strap_MSTR, loc=<14.6337 18.6178 -0.0032>, rot=<0.7682 -0.6402 -0.0024 0.0035>}
]
