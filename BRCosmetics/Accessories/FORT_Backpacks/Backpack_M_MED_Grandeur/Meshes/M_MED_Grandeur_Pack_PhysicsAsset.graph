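# Physics asset for the M_MED_Grandeur backpack cosmetic, apparently for secondary motion of its dangling top strap.
# Capsule bodies cover the pack, a tiny strap master bone and two top strap segments.
# The master rotates freely against the pack, the first segment is tightly limited to the master, and the second swings more widely from the first.
# An extra constraint links the second segment directly to the pack, closing the chain into a loop.
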PhysicsAsset /BRCosmetics/Accessories/FORT_Backpacks/Backpack_M_MED_Grandeur/Meshes/M_MED_Grandeur_Pack_PhysicsAsset
bodies {
 dyn_strap_MSTR [{shapeType=Capsule, length=0.505, radius=0.505, shapeName=dyn_strap_MSTR_capsule, rotation=<rot 0 -90 0>}]
 dyn_strap_top_1 [{shapeType=Capsule, length=8, radius=2, shapeName=dyn_strap_top_1_capsule, center=<5 0 0>, rotation=<rot 90 0 0>}]
 dyn_strap_top_2 [{shapeType=Capsule, length=8, radius=2, shapeName=dyn_strap_top_2_capsule, center=<5 0 0>, rotation=<rot 90 0 0>}]
 pack [{
   shapeType=Capsule
   length=38.7292
   radius=2.6289
   shapeName=pack_capsule
   center=<0.1808 -3.2692 -0.3933>
   rotation=<rot -0.0079 -90 39.2494>
  }]
}
constraints [
 {
  bone1Name=dyn_strap_MSTR
  bone2Name=pack
  swing1LimitDegrees=45
  swing1Motion=Free
  swing2LimitDegrees=45
  swing2Motion=Free
  twistLimitDegrees=45
  twistMotion=Free
 }
 {bone1Name=dyn_strap_top_1, bone2Name=dyn_strap_MSTR, swing1LimitDegrees=5, swing2LimitDegrees=20, twistLimitDegrees=1}
 {bone1Name=dyn_strap_top_2, bone2Name=dyn_strap_top_1, swing1LimitDegrees=75, swing2LimitDegrees=75, twistLimitDegrees=15}
 {bone1Name=dyn_strap_top_2, bone2Name=pack, swing1LimitDegrees=75, swing2LimitDegrees=75, twistLimitDegrees=45}
]
